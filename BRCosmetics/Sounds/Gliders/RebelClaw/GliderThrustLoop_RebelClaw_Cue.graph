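# Looping thrust sound cue for the RebelClaw glider cosmetic, using player glider thrust attenuation and the glider loop audio property sheet.
# A mixer blends six looping layers, each with continuous pitch and volume modulation driven by movement parameters.
# AnyDirection drives engine, wind, high and mech layers, while Forwards drives flame and electric.
# The engine layer's volume range is inverted, so it ducks as AnyDirection rises while the other layers get louder.
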SoundCue /BRCosmetics/Sounds/Gliders/RebelClaw/GliderThrustLoop_RebelClaw_Cue
root=Mixer_0
nodes [
 {
  id=Mixer_0
  children [ModulatorContinuous_0, ModulatorContinuous_1, ModulatorContinuous_2, ModulatorContinuous_3, ModulatorContinuous_4, ModulatorContinuous_5]
  properties {inputVolume [0.7,0.5,0.7,0.9,0.6,0.8]}
 }
 {
  id=ModulatorContinuous_0
  children [WavePlayer_0]
  properties {
   pitchModulationParams {default=0, maxOutput=1.1, minOutput=1, parameterName=AnyDirection}
   volumeModulationParams {default=0, maxOutput=0.4, minOutput=0.6, parameterName=AnyDirection}
  }
 }
 {
  id=ModulatorContinuous_1
  children [WavePlayer_1]
  properties {
   pitchModulationParams {default=0, maxOutput=1.1, minOutput=1, parameterName=AnyDirection}
   volumeModulationParams {default=0, maxOutput=0.8, minOutput=0.5, parameterName=AnyDirection}
  }
 }
 {
  id=ModulatorContinuous_2
  children [WavePlayer_2]
  properties {
   pitchModulationParams {default=0, maxOutput=1.1, minOutput=1, parameterName=AnyDirection}
   volumeModulationParams {default=0, minOutput=0.3, parameterName=AnyDirection}
  }
 }
 {
  id=ModulatorContinuous_3
  children [WavePlayer_3]
  properties {pitchModulationParams {default=0, minOutput=0.9, parameterName=Forwards}, volumeModulationParams {default=0, parameterName=Forwards}}
 }
 {
  id=ModulatorContinuous_4
  children [WavePlayer_4]
  properties {
   pitchModulationParams {default=0, maxOutput=1.2, minOutput=1, parameterName=Forwards}
   volumeModulationParams {default=0, maxOutput=0.5, parameterName=Forwards}
  }
 }
 {
  id=ModulatorContinuous_5
  children [WavePlayer_5]
  properties {
   pitchModulationParams {default=0, minOutput=0.98, parameterName=AnyDirection}
   volumeModulationParams {default=0, maxOutput=0.8, minOutput=0.4, parameterName=AnyDirection}
  }
 }
 {id=WavePlayer_0, properties {bLooping=true, soundWaveAssetPtr=&/Glider_RebelClaw_Thurst_LP_Engine}}
 {id=WavePlayer_1, properties {bLooping=true, soundWaveAssetPtr=&/Glider_RebelClaw_Thurst_LP_Wind}}
 {id=WavePlayer_2, properties {bLooping=true, soundWaveAssetPtr=&/Glider_RebelClaw_Thurst_LP_High}}
 {id=WavePlayer_3, properties {bLooping=true, soundWaveAssetPtr=&/Glider_RebelClaw_Thurst_LP_Flame}}
 {id=WavePlayer_4, properties {bLooping=true, soundWaveAssetPtr=&/Glider_RebelClaw_Thurst_LP_Electric}}
 {id=WavePlayer_5, properties {bLooping=true, soundWaveAssetPtr=&/Glider_RebelClaw_Thurst_LP_Mech}}
]
properties {
 attenuationSettings=&Fort_Player_Glider_Thrust_Athena_Att
 audioPropertiesSheet=&APS_Cosmetic_Glider_Loop
 maxDistance=12500
 soundClassObject=&Fort_Gliders_Land
 volumeMultiplier=0.6
 concurrencySet [null]
}
